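# Unlit, two-sided translucent mesh-particle material for the idle lightning effect on the Boardwalk pickaxe.
# Color blends two tints through a sharpened mask peaking mid-U, multiplies by particle color, then passes through MF_FX_EmissiveAdjustments.
# A panned normal-noise texture drives a local-space wobble offset, sine-tapered along V to pin both ends.
# Dynamic parameters control noise speed, lightning complexity and randomized wobble strength.
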Material /BRCosmetics/Weapons/FORT_Melee/Pickaxe_AshtonBoardwalk/FX/M_Boardwalk_Pickaxe_Ilde_lightning
graph {
 properties {
  bUsedWithMeshParticles=true
  bUsedWithNiagaraMeshParticles=true
  blendMode=BLEND_Translucent
  pixelDepthOffsetMode=PDOM_Legacy
  shadingModel=MSM_Unlit
  twoSided=true
  usedShadingModels=Unlit
 }
 nodes [
  n1 = VectorParameter() {parameterName=Color_1, defaultValue=<rgba 0.1818 0.5709 1 1>}
  n2 = VectorParameter() {parameterName=Color_2, defaultValue=<rgba 0.8177 0.9942 1 1>}
  n3 = ComponentMask(None=TextureCoordinate()) {r=true}
  n4 = Multiply(A=OneMinus(None=@n3), B=@n3)
  n5 = ScalarParameter() {defaultValue=3, group=Color, parameterName=Mask Exponent}
  n6 = Clamp(None=Multiply(A=Power(Base=@n4, Exp=@n5) {constExponent=3}, B=ScalarParameter() {defaultValue=50, group=Color, parameterName=Mask Scale}) {constB=50})
  MaterialFunctionCall(Emissive=Multiply(A=ParticleColor().RGB, B=LinearInterpolate(A=@n1.RGB, B=@n2.RGB, Alpha=@n6))) {materialFunction=&MF_FX_EmissiveAdjustments}
  n7 = TextureCoordinate()
  n8 = ComponentMask(None=@n7) {g=true}
  n9 = DynamicParameter() {paramNames ["WithRandom(0.8-1)", NoiseSpeedMult, "Lightning_complexity(0-1)", Param4]}
  n10 = AppendVector(A=Constant() {r=0.02}, B=@n9."Lightning_complexity(0-1)")
  n11 = Constant2Vector() {g=0.02, r=0.7}
  n12 = TextureSample!(
   UVs=Panner(Coordinate=Multiply(A=@n7, B=@n10), Speed=Multiply(A=@n11, B=@n9.NoiseSpeedMult)) {speedX=1, speedY=0.2}
   Tex
   Level
   Apply View MipBias
  ) {mipValueMode=TMVM_MipLevel, samplerType=SAMPLERTYPE_Normal, texture=/Game/Global/Textures/Default/Noise/T_Normal_Noise01}
  n13 = TextureCoordinate()
  n14 = Multiply(A=LinearInterpolate(A=@n9."WithRandom(0.8-1)", Alpha=ComponentMask(None=@n13) {r=true}) {constA=0.6}) {constB=3}
  n15 = VectorParameter() {group=WPO, parameterName=WobbleOffset, defaultValue=<rgba 1 1 50 1>}
  Multiply(
   A=Sine(None=Divide(A=@n8))
   B=Multiply(A=Multiply(A=@n12.RGB, B=@n14) {constB=10}, B=Transform(None=@n15.RGB) {transformSourceType=TRANSFORMSOURCE_Local}) {constB=2}
  )
 ]
}
