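# Unlit glow material for the Ashton Boardwalk pickaxe cosmetic, usable on skeletal meshes and with mesh deformers.
# A BaseColor vector parameter, defaulting to a bright HDR blue, drives emissive color through material attributes.
# The attributes then pass through the shared MF_TechArtOverrides_accessories function.
# Despite the MI_ prefix, the asset is a base material rather than an instance.
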Material /BRCosmetics/Weapons/FORT_Melee/Pickaxe_AshtonBoardwalk/Materials/MI_Pickaxe_Ashton_Boardwalk_GLOW2
graph {
 properties {
  bUseMaterialAttributes=true
  bUsedWithMeshDeformer=true
  bUsedWithSkeletalMesh=true
  pixelDepthOffsetMode=PDOM_Legacy
  shadingModel=MSM_Unlit
  usedShadingModels=Unlit
 }
 nodes [MaterialFunctionCall(In=SetMaterialAttributes(Emissive Color=VectorParameter() {parameterName=BaseColor, defaultValue=<rgba 0 5 10 1>}.RGB) {attributeSetTypes [#1]}) {materialFunction=&MF_TechArtOverrides_accessories}]
}
