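# Physics asset for the Ashton Boardwalk two-handed pickaxe cosmetic, giving the handle and two dangling ribbons collision bodies and joints.
# The handle is a single long capsule that roots two three-bone ribbon chains, a and b.
# Each chain starts with a thin box and continues with two small capsules.
# Every joint links a segment to its parent with equal swing and twist limit angles.
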PhysicsAsset /BRCosmetics/Weapons/FORT_Melee/Pickaxe_AshtonBoardwalk/Meshes/Pickaxe_Ashton_Boardwalk_Physics
bodies {
 Melee_TwoHanded_Handle [{shapeType=Capsule, length=125.2754, radius=2.2776, center=<1.4004 -1.0262 27.4231>}]
 dyn_ribbon_a01 [{shapeType=Box, extentX=4.7045, extentY=0.5557, extentZ=1.7763, center=<1.9396 0 0>}]
 dyn_ribbon_a02 [{shapeType=Capsule, length=5.1162, radius=1.2108, center=<1.5 -0.0002 0>, rotation=<rot -90 0.844 -0.8441>}]
 dyn_ribbon_a03 [{shapeType=Capsule, length=4.8868, radius=1.4403, center=<1.4319 -0.0002 0>, rotation=<rot -90 -1.8984 1.8983>}]
 dyn_ribbon_b01 [{shapeType=Box, extentX=4.6981, extentY=0.5557, extentZ=1.7763, center=<2.3407 0 0>}]
 dyn_ribbon_b02 [{shapeType=Capsule, length=3.8719, radius=1.2108, center=<1.3626 -0.0901 -0.0055>, rotation=<rot -90 5.0511 -5.0511>}]
 dyn_ribbon_b03 [{shapeType=Capsule, length=3.9219, radius=1.4403, center=<1.3582 -0.09 -0.0055>, rotation=<rot -90 5.0511 -5.0511>}]
}
constraints [
 {bone1Name=dyn_ribbon_a01, bone2Name=Melee_TwoHanded_Handle, swing1LimitDegrees=45, swing2LimitDegrees=45, twistLimitDegrees=45}
 {bone1Name=dyn_ribbon_a02, bone2Name=dyn_ribbon_a01, swing1LimitDegrees=45, swing2LimitDegrees=45, twistLimitDegrees=45}
 {bone1Name=dyn_ribbon_a03, bone2Name=dyn_ribbon_a02, swing1LimitDegrees=45, swing2LimitDegrees=45, twistLimitDegrees=45}
 {bone1Name=dyn_ribbon_b01, bone2Name=Melee_TwoHanded_Handle, swing1LimitDegrees=45, swing2LimitDegrees=45, twistLimitDegrees=45}
 {bone1Name=dyn_ribbon_b02, bone2Name=dyn_ribbon_b01, swing1LimitDegrees=45, swing2LimitDegrees=45, twistLimitDegrees=45}
 {bone1Name=dyn_ribbon_b03, bone2Name=dyn_ribbon_b02, swing1LimitDegrees=45, swing2LimitDegrees=45, twistLimitDegrees=45}
]
